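# Looping indoor ambience cue for a wet underbridge area in the Underworld biome, played quietly with ambient attenuation.
# Each branch mixes flowing cave river, cave drips and wet castle-basement beds at different levels.
# Two nested looping envelopers of different lengths add slow pitch movement and a gentle volume swell, varying the loop.
# A quality switch selects the quad basement bed for its first two slots and the stereo version for the last two.
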SoundCue /Game/Sounds/Helios/Biomes/Underworld/Ambience_Underworld_Underbridge_Wet_Loop_Cue
root=QualityLevel_0
nodes [
 {
  id=Enveloper_1
  children [Mixer_0]
  properties {
   bLoop=true
   bLoopIndefinitely=true
   loopEnd=30
   pitchCurve {editorCurveData {keys [{interpMode=RCIM_Cubic, time=0}, <at 15.9468 0.9181 RCIM_Cubic>, {interpMode=RCIM_Cubic, time=30}]}}
   volumeCurve {editorCurveData {keys [<at 0 1>]}}
  }
 }
 {
  id=Enveloper_2
  children [Enveloper_1]
  properties {
   bLoop=true
   bLoopIndefinitely=true
   loopEnd=20
   pitchCurve {editorCurveData {keys [{time=0}]}}
   volumeCurve {editorCurveData {keys [
      {arriveTangent=0, interpMode=RCIM_Cubic, leaveTangent=0, time=0, value=1}
      {arriveTangent=0, interpMode=RCIM_Cubic, leaveTangent=0, time=10, value=0.9}
      {arriveTangent=0, interpMode=RCIM_Cubic, leaveTangent=0, time=20, value=0.9992}
     ]}}
  }
 }
 {
  id=Enveloper_4
  children [Mixer_1]
  properties {
   bLoop=true
   bLoopIndefinitely=true
   loopEnd=30
   pitchCurve {editorCurveData {keys [{interpMode=RCIM_Cubic, time=0}, <at 15.9468 0.9181 RCIM_Cubic>, {interpMode=RCIM_Cubic, time=30}]}}
   volumeCurve {editorCurveData {keys [<at 0 1>]}}
  }
 }
 {
  id=Enveloper_5
  children [Enveloper_4]
  properties {
   bLoop=true
   bLoopIndefinitely=true
   loopEnd=20
   pitchCurve {editorCurveData {keys [{time=0}]}}
   volumeCurve {editorCurveData {keys [
      {arriveTangent=0, interpMode=RCIM_Cubic, leaveTangent=0, time=0, value=1}
      {arriveTangent=0, interpMode=RCIM_Cubic, leaveTangent=0, time=10, value=0.9}
      {arriveTangent=0, interpMode=RCIM_Cubic, leaveTangent=0, time=20, value=0.9992}
     ]}}
  }
 }
 {id=Mixer_0, children [WavePlayer_3,WavePlayer_2,WavePlayer_0], properties {inputVolume [0.4,0.8,1]}}
 {id=Mixer_1, children [WavePlayer_5,WavePlayer_4,WavePlayer_1], properties {inputVolume [0.4,0.8,1]}}
 {id=QualityLevel_0, children [Enveloper_2,Enveloper_2,Enveloper_5,Enveloper_5]}
 {id=WavePlayer_0, properties {bLooping=true, soundWaveAssetPtr=/Game/Sounds/Asteria/POI/Knights/Ambience_Knights_Castle_Basement_03_Wet_Quad}}
 {
  id=WavePlayer_1
  properties {bLooping=true, soundWaveAssetPtr=/Game/Sounds/Asteria/POI/Knights/Ambience_Knights_Castle_Basement_03_Wet_Stereo}
 }
 {id=WavePlayer_2, properties {bLooping=true, soundWaveAssetPtr=/Game/Sounds/Fort_Ambient_Looping/Cave/fort_cave_drips_01}}
 {id=WavePlayer_3, properties {bLooping=true, soundWaveAssetPtr=/Game/Sounds/Fort_Ambient_Looping/Water/Spot_Water_Cave_River_Flowing}}
 {id=WavePlayer_4, properties {bLooping=true, soundWaveAssetPtr=&fort_cave_drips_01}}
 {id=WavePlayer_5, properties {bLooping=true, soundWaveAssetPtr=&Spot_Water_Cave_River_Flowing}}
]
properties {
 attenuationSettings=&Fort_Ambient_Athena_Att
 maxDistance=2097152
 soundClassObject=&Fort_Ambient_Indoors_Athena
 volumeMultiplier=0.15
 preEffectBusSends [{}]
}
